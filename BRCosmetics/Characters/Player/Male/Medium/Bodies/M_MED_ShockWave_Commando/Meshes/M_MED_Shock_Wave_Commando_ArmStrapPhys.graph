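# Physics asset for the ShockWave Commando male medium body's arm-strap rig, covering dynamic torso tubes and a back cable.
# Left and right tube chains run from master bodies on spine_05 through three capsule segments with tight twist and narrow secondary swing limits.
# The back cable and large upper-arm boxes, likely colliders, also constrain to spine_05.
# Both arms and the right master have free angular motion, whereas the left master is limited.
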PhysicsAsset /BRCosmetics/Characters/Player/Male/Medium/Bodies/M_MED_ShockWave_Commando/Meshes/M_MED_Shock_Wave_Commando_ArmStrapPhys
bodies {
 dyn_back_cable [{shapeType=Capsule, length=2.7532, radius=2.2368, center=<0 0 0.884>, rotation=<rot 0 -90 0>}]
 dyn_torso_l_tube_1 [{shapeType=Capsule, length=9.278, radius=3.1522, rotation=<rot 0 90 0>}]
 dyn_torso_l_tube_2 [{shapeType=Capsule, length=9.278, radius=3.1522, rotation=<rot 0 90 0>}]
 dyn_torso_l_tube_3 [{shapeType=Capsule, length=8.778, radius=6.4022, center=<0 -4.5491 0>, rotation=<rot 0 90 0>}]
 dyn_torso_l_tube_mstr1 [{shapeType=Capsule, length=7.25, radius=7, center=<7.1843 -16.5406 0.885>, rotation=<rot 90 180 -173.1975>}]
 dyn_torso_r_tube_1 [{shapeType=Capsule, length=9.278, radius=3.1522, rotation=<rot 0 90 0>}]
 dyn_torso_r_tube_2 [{shapeType=Capsule, length=9.278, radius=3.1522, rotation=<rot 0 90 0>}]
 dyn_torso_r_tube_3 [{shapeType=Capsule, length=9.278, radius=6.4022, center=<0 4.6701 0>, rotation=<rot 8.7268 90 -0.0001>}]
 dyn_torso_r_tube_mstr [{shapeType=Capsule, length=20.5, radius=6.25, center=<2.4534 15.3382 0.8853>, rotation=<rot -0.0003 -7.9041 -90>}]
 spine_05 [{shapeType=Capsule, length=7.75, radius=12.9006, rotation=<rot 0 -90 0>}]
 upperarm_l [{shapeType=Box, extentX=21, extentY=16, extentZ=18.5, center=<11.6339 2.0796 6.8071>, rotation=<rot 0.0001 0 4.4547>}]
 upperarm_r [{shapeType=Box, extentX=21, extentY=16, extentZ=18.5, center=<-11.6339 -2.0796 -6.8071>, rotation=<rot 0.0001 0 -0.1311>}]
}
constraints [
 {
  bone1Name=dyn_torso_r_tube_mstr
  bone2Name=spine_05
  swing1LimitDegrees=45
  swing1Motion=Free
  swing2LimitDegrees=45
  swing2Motion=Free
  twistLimitDegrees=45
  twistMotion=Free
 }
 {bone1Name=dyn_torso_l_tube_mstr1, bone2Name=spine_05, swing1LimitDegrees=45, swing2LimitDegrees=2, twistLimitDegrees=1}
 {bone1Name=dyn_torso_r_tube_2, bone2Name=dyn_torso_r_tube_1, swing1LimitDegrees=20, swing2LimitDegrees=3, twistLimitDegrees=1}
 {bone1Name=dyn_torso_r_tube_3, bone2Name=dyn_torso_r_tube_2, swing1LimitDegrees=30, swing2LimitDegrees=5, twistLimitDegrees=1}
 {bone1Name=dyn_torso_r_tube_1, bone2Name=dyn_torso_r_tube_mstr, swing1LimitDegrees=20, swing2LimitDegrees=1, twistLimitDegrees=1}
 {
  bone1Name=upperarm_r
  bone2Name=spine_05
  swing1LimitDegrees=45
  swing1Motion=Free
  swing2LimitDegrees=45
  swing2Motion=Free
  twistLimitDegrees=45
  twistMotion=Free
 }
 {bone1Name=dyn_torso_l_tube_1, bone2Name=dyn_torso_l_tube_mstr1, swing1LimitDegrees=20, swing2LimitDegrees=1, twistLimitDegrees=1}
 {bone1Name=dyn_torso_l_tube_2, bone2Name=dyn_torso_l_tube_1, swing1LimitDegrees=20, swing2LimitDegrees=3, twistLimitDegrees=1}
 {bone1Name=dyn_torso_l_tube_3, bone2Name=dyn_torso_l_tube_2, swing1LimitDegrees=30, swing2LimitDegrees=5, twistLimitDegrees=5}
 {
  bone1Name=upperarm_l
  bone2Name=spine_05
  swing1LimitDegrees=45
  swing1Motion=Free
  swing2LimitDegrees=45
  swing2Motion=Free
  twistLimitDegrees=45
  twistMotion=Free
 }
 {bone1Name=dyn_back_cable, bone2Name=spine_05, swing1LimitDegrees=45, swing2LimitDegrees=45, twistLimitDegrees=1}
]
