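# Physics asset for the male medium ShockWave Commando player body, covering its dangling hip tubes, torso tubes and back cable.
# Coarse pelvis, upper-spine, thigh and upper-arm bodies form the core, linked by free-angular joints.
# Angle-limited chains make up each tube, with torso tubes rooted at spine_05 and hip tubes at pelvis or master bones.
# Free cross-constraints tie front and back hip chains near their tips, closing each side into a loop.
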PhysicsAsset /BRCosmetics/Characters/Player/Male/Medium/Bodies/M_MED_ShockWave_Commando/Meshes/M_MED_Shock_Wave_Commando_Physics
bodies {
 dyn_back_cable [{shapeType=Capsule, length=2.7532, radius=2.2368, center=<0 0 0.884>, rotation=<rot 0 -90 0>}]
 dyn_hip_tube_bk_l_01 [{shapeType=Box, extentX=4, extentY=4.2, extentZ=6.7261, center=<1.5901 -0.0125 -0.5527>, rotation=<rot 1.1874 90.3227 -85.5629>}]
 dyn_hip_tube_bk_l_02 [{shapeType=Capsule, length=4.5, radius=2.25, center=<2.8756 -0.03 -0.2133>, rotation=<rot 90 0 0>}]
 dyn_hip_tube_bk_l_03 [{shapeType=Capsule, length=4.5, radius=2.25, center=<2.8756 -0.0299 -0.279>, rotation=<rot 90 0 0>}]
 dyn_hip_tube_bk_l_04 [{shapeType=Capsule, length=4.5, radius=2.25, center=<2.8756 -0.03 0.0607>, rotation=<rot 90 0 0>}]
 dyn_hip_tube_bk_l_05 [{shapeType=Box, extentX=0.5, extentY=1.7, extentZ=3.2261, center=<-0.4687 -0.0207 -0.7125>, rotation=<rot -85.405 15.2758 -14.999>}]
 dyn_hip_tube_bk_r_01 [{shapeType=Box, extentX=4, extentY=4.2, extentZ=6.7261, center=<1.5901 -0.0125 -0.5527>, rotation=<rot 1.1874 90.3227 -85.5629>}]
 dyn_hip_tube_bk_r_02 [{shapeType=Capsule, length=4.5, radius=2.25, center=<2.8756 -0.03 0.2398>, rotation=<rot 90 0 0>}]
 dyn_hip_tube_bk_r_03 [{shapeType=Capsule, length=4.5, radius=2.25, center=<2.8756 -0.0299 0.1883>, rotation=<rot 90 0 0>}]
 dyn_hip_tube_bk_r_04 [{shapeType=Capsule, length=4.5, radius=2.25, center=<2.8756 -0.0299 0.1095>, rotation=<rot 90 0 0>}]
 dyn_hip_tube_bk_r_05 [{shapeType=Box, extentX=0.5, extentY=1.7, extentZ=3.2261, center=<-0.4687 -0.0207 -0.7125>, rotation=<rot -85.405 15.2758 -14.999>}]
 dyn_hip_tube_bk_r_MSTR [{shapeType=Capsule, length=0.505, radius=0.505, shapeName=dyn_hip_tube_bk_r_MSTR_capsule, rotation=<rot 0 -90 0>}]
 dyn_hip_tube_ft_l_01 [{shapeType=Capsule, length=4.5, radius=2.25, center=<2.8756 -0.0299 0.1413>, rotation=<rot 90 0 0>}]
 dyn_hip_tube_ft_l_02 [{shapeType=Capsule, length=4.5, radius=2.25, center=<2.8756 -0.0299 -0.0186>, rotation=<rot 90 0 0>}]
 dyn_hip_tube_ft_l_03 [{shapeType=Capsule, length=4.5, radius=2.25, center=<2.8756 -0.0299 0.0798>, rotation=<rot 90 0 0>}]
 dyn_hip_tube_ft_l_04 [{shapeType=Capsule, length=4.5, radius=2.25, center=<2.8757 -0.03 0.2341>, rotation=<rot 90 0 0>}]
 dyn_hip_tube_ft_l_05 [{shapeType=Box, extentX=1.01, extentY=1.01, extentZ=0.5, rotation=<rot 0 -90 0>}]
 dyn_hip_tube_ft_r_01 [{shapeType=Capsule, length=4.5, radius=2.25, center=<2.8756 -0.0299 -0.0992>, rotation=<rot 90 0 0>}]
 dyn_hip_tube_ft_r_02 [{shapeType=Capsule, length=4.5, radius=2.25, center=<2.8756 -0.0299 -0.007>, rotation=<rot 90 0 0>}]
 dyn_hip_tube_ft_r_03 [{shapeType=Capsule, length=4.5, radius=2.25, center=<2.8756 -0.0299 0.0118>, rotation=<rot 90 0 0>}]
 dyn_hip_tube_ft_r_04 [{shapeType=Capsule, length=4.5, radius=2.25, center=<2.8756 -0.0299 0.0443>, rotation=<rot 90 0 0>}]
 dyn_hip_tube_ft_r_05 [{shapeType=Box, extentX=1.01, extentY=1.01, extentZ=0.5, rotation=<rot 0 -90 0>}]
 dyn_torso_l_tube_1 [{shapeType=Capsule, length=9.278, radius=3.1522, rotation=<rot 0 90 0>}]
 dyn_torso_l_tube_2 [{shapeType=Capsule, length=9.278, radius=3.1522, rotation=<rot 0 90 0>}]
 dyn_torso_l_tube_3 [{shapeType=Capsule, length=9.278, radius=2.6522, center=<0 -0.0477 0>, rotation=<rot 0 90 0>}]
 dyn_torso_l_tube_mstr1 [{shapeType=Capsule, length=7, radius=7.25, center=<7.0606 -9.5126 0.885>, rotation=<rot 90 0 0.0051>}]
 dyn_torso_r_tube_1 [{shapeType=Capsule, length=9.278, radius=3.1522, rotation=<rot 0 90 0>}]
 dyn_torso_r_tube_2 [{shapeType=Capsule, length=9.278, radius=3.1522, rotation=<rot 0 90 0>}]
 dyn_torso_r_tube_3 [{shapeType=Capsule, length=9.278, radius=3.1522, center=<0 -0.0477 0>, rotation=<rot 0 90 0>}]
 dyn_torso_r_tube_mstr [{shapeType=Capsule, length=6, radius=8.25, center=<-7.1603 11.4721 0.8853>, rotation=<rot 90 -0.4686 -0.4635>}]
 dyn_tube_front_l_MSTR [{shapeType=Capsule, length=0.505, radius=0.505, rotation=<rot 0 -90 0>}]
 dyn_tube_front_r_MSTR [{shapeType=Capsule, length=0.505, radius=0.505, rotation=<rot 0 -90 0>}]
 dyn_waist_tube_r_MSTR [{shapeType=Capsule, length=0.505, radius=0.505, rotation=<rot 0 -90 0>}]
 pelvis [{shapeType=Capsule, length=9.75, radius=10.2024, center=<-8.785 -0.8103 0.4534>, rotation=<rot 2.5393 -89.9995 0.1151>}]
 spine_05 [{shapeType=Capsule, length=7.75, radius=12.9006, rotation=<rot 0 -90 0>}]
 thigh_l [{shapeType=Capsule, length=19.1593, radius=9.5, center=<-14.6185 0.2692 -0.2453>, rotation=<rot -89.5403 -91.9277 86.9433>}]
 thigh_r [{shapeType=Capsule, length=19.4093, radius=9.5, center=<15.1058 -1.0544 0.31>, rotation=<rot 59.5657 85.0542 90.0438>}]
 upperarm_l [{shapeType=Box, extentX=21, extentY=16, extentZ=18.5, center=<11.6339 2.0796 6.8071>, rotation=<rot 0.0002 0 20.0031>}]
 upperarm_r [
  {shapeType=Box, extentX=18.5, extentY=16, extentZ=12.5, center=<-8.7958 0.0006 -1.7849>, rotation=<rot 0.0001 0 0>}
  {shapeType=Box, extentX=18.5, extentY=16, extentZ=12.5, center=<-7.7266 0.0007 -5.1865>, rotation=<rot 0.0001 0 0>}
 ]
}
constraints [
 {
  bone1Name=thigh_l
  bone2Name=pelvis
  swing1LimitDegrees=45
  swing1Motion=Free
  swing2LimitDegrees=45
  swing2Motion=Free
  twistLimitDegrees=45
  twistMotion=Free
 }
 {
  bone1Name=thigh_r
  bone2Name=pelvis
  swing1LimitDegrees=45
  swing1Motion=Free
  swing2LimitDegrees=45
  swing2Motion=Free
  twistLimitDegrees=45
  twistMotion=Free
 }
 {
  bone1Name=dyn_torso_r_tube_mstr
  bone2Name=spine_05
  swing1LimitDegrees=45
  swing1Motion=Free
  swing2LimitDegrees=45
  swing2Motion=Free
  twistLimitDegrees=45
  twistMotion=Free
 }
 {bone1Name=dyn_torso_l_tube_mstr1, bone2Name=spine_05, swing1LimitDegrees=45, swing2LimitDegrees=2, twistLimitDegrees=1}
 {bone1Name=dyn_torso_r_tube_2, bone2Name=dyn_torso_r_tube_1, swing1LimitDegrees=45, swing2LimitDegrees=2, twistLimitDegrees=1}
 {bone1Name=dyn_torso_r_tube_3, bone2Name=dyn_torso_r_tube_2, swing1LimitDegrees=45, swing2LimitDegrees=2, twistLimitDegrees=1}
 {bone1Name=dyn_torso_r_tube_1, bone2Name=dyn_torso_r_tube_mstr, swing1LimitDegrees=45, swing2LimitDegrees=2, twistLimitDegrees=1}
 {
  bone1Name=upperarm_r
  bone2Name=spine_05
  swing1LimitDegrees=45
  swing1Motion=Free
  swing2LimitDegrees=45
  swing2Motion=Free
  twistLimitDegrees=45
  twistMotion=Free
 }
 {bone1Name=dyn_torso_l_tube_1, bone2Name=dyn_torso_l_tube_mstr1, swing1LimitDegrees=45, swing2LimitDegrees=1, twistLimitDegrees=1}
 {bone1Name=dyn_torso_l_tube_2, bone2Name=dyn_torso_l_tube_1, swing1LimitDegrees=45, swing2LimitDegrees=1, twistLimitDegrees=1}
 {bone1Name=dyn_torso_l_tube_3, bone2Name=dyn_torso_l_tube_2, swing1LimitDegrees=45, swing2LimitDegrees=1, twistLimitDegrees=1}
 {
  bone1Name=upperarm_l
  bone2Name=spine_05
  swing1LimitDegrees=45
  swing1Motion=Free
  swing2LimitDegrees=45
  swing2Motion=Free
  twistLimitDegrees=45
  twistMotion=Free
 }
 {bone1Name=dyn_back_cable, bone2Name=spine_05, swing1LimitDegrees=45, swing2LimitDegrees=45, twistLimitDegrees=1}
 {
  bone1Name=spine_05
  bone2Name=pelvis
  swing1LimitDegrees=45
  swing1Motion=Free
  swing2LimitDegrees=45
  swing2Motion=Free
  twistLimitDegrees=45
  twistMotion=Free
 }
 {bone1Name=dyn_hip_tube_bk_l_05, bone2Name=dyn_hip_tube_bk_l_04, swing1LimitDegrees=60, swing2LimitDegrees=60, twistLimitDegrees=20}
 {bone1Name=dyn_hip_tube_bk_l_04, bone2Name=dyn_hip_tube_bk_l_03, swing1LimitDegrees=60, swing2LimitDegrees=60, twistLimitDegrees=20}
 {bone1Name=dyn_hip_tube_bk_l_03, bone2Name=dyn_hip_tube_bk_l_02, swing1LimitDegrees=60, swing2LimitDegrees=60, twistLimitDegrees=20}
 {bone1Name=dyn_hip_tube_bk_l_02, bone2Name=dyn_hip_tube_bk_l_01, swing1LimitDegrees=60, swing2LimitDegrees=60, twistLimitDegrees=20}
 {bone1Name=dyn_hip_tube_bk_l_01, bone2Name=pelvis, swing1LimitDegrees=45, swing2LimitDegrees=45, twistLimitDegrees=20}
 {bone1Name=dyn_hip_tube_ft_l_05, bone2Name=dyn_hip_tube_ft_l_04, swing1LimitDegrees=60, swing2LimitDegrees=60, twistLimitDegrees=20}
 {bone1Name=dyn_hip_tube_ft_l_04, bone2Name=dyn_hip_tube_ft_l_03, swing1LimitDegrees=60, swing2LimitDegrees=60, twistLimitDegrees=20}
 {bone1Name=dyn_hip_tube_ft_l_03, bone2Name=dyn_hip_tube_ft_l_02, swing1LimitDegrees=60, swing2LimitDegrees=60, twistLimitDegrees=20}
 {bone1Name=dyn_hip_tube_ft_l_02, bone2Name=dyn_hip_tube_ft_l_01, swing1LimitDegrees=60, swing2LimitDegrees=60, twistLimitDegrees=20}
 {bone1Name=dyn_hip_tube_ft_l_01, bone2Name=dyn_tube_front_l_MSTR, swing1LimitDegrees=60, swing2LimitDegrees=60, twistLimitDegrees=20}
 {
  bone1Name=dyn_hip_tube_bk_l_05
  bone2Name=dyn_hip_tube_ft_l_05
  swing1LimitDegrees=45
  swing1Motion=Free
  swing2LimitDegrees=45
  swing2Motion=Free
  twistLimitDegrees=45
  twistMotion=Free
 }
 {bone1Name=dyn_hip_tube_bk_r_04, bone2Name=dyn_hip_tube_bk_r_03, swing1LimitDegrees=60, swing2LimitDegrees=60, twistLimitDegrees=1}
 {bone1Name=dyn_hip_tube_bk_r_01, bone2Name=dyn_hip_tube_bk_r_MSTR, swing1LimitDegrees=45, swing2LimitDegrees=45, twistLimitDegrees=1}
 {bone1Name=dyn_hip_tube_bk_r_02, bone2Name=dyn_hip_tube_bk_r_01, swing1LimitDegrees=45, swing2LimitDegrees=45, twistLimitDegrees=1}
 {bone1Name=dyn_hip_tube_bk_r_03, bone2Name=dyn_hip_tube_bk_r_02, swing1LimitDegrees=45, swing2LimitDegrees=45, twistLimitDegrees=1}
 {bone1Name=dyn_waist_tube_r_MSTR, bone2Name=pelvis, swing1LimitDegrees=60, swing2LimitDegrees=60, twistLimitDegrees=20}
 {bone1Name=dyn_tube_front_r_MSTR, bone2Name=dyn_waist_tube_r_MSTR, swing1LimitDegrees=60, swing2LimitDegrees=60, twistLimitDegrees=20}
 {bone1Name=dyn_hip_tube_ft_r_01, bone2Name=dyn_tube_front_r_MSTR, swing1LimitDegrees=45, swing2LimitDegrees=45, twistLimitDegrees=1}
 {bone1Name=dyn_hip_tube_ft_r_02, bone2Name=dyn_hip_tube_ft_r_01, swing1LimitDegrees=45, swing2LimitDegrees=45, twistLimitDegrees=1}
 {bone1Name=dyn_hip_tube_ft_r_03, bone2Name=dyn_hip_tube_ft_r_02, swing1LimitDegrees=45, swing2LimitDegrees=45, twistLimitDegrees=1}
 {bone1Name=dyn_hip_tube_ft_r_04, bone2Name=dyn_hip_tube_ft_r_03, swing1LimitDegrees=45, swing2LimitDegrees=45, twistLimitDegrees=1}
 {bone1Name=dyn_hip_tube_ft_r_05, bone2Name=dyn_hip_tube_ft_r_04, swing1LimitDegrees=60, swing2LimitDegrees=60, twistLimitDegrees=20}
 {
  bone1Name=dyn_hip_tube_bk_r_05
  bone2Name=dyn_hip_tube_bk_r_04
  swing1LimitDegrees=45
  swing1Motion=Free
  swing2LimitDegrees=45
  swing2Motion=Free
  twistLimitDegrees=45
  twistMotion=Free
 }
 {
  bone1Name=dyn_hip_tube_bk_r_05
  bone2Name=dyn_hip_tube_ft_r_05
  swing1LimitDegrees=45
  swing1Motion=Free
  swing2LimitDegrees=45
  swing2Motion=Free
  twistLimitDegrees=45
  twistMotion=Free
 }
 {
  bone1Name=dyn_hip_tube_bk_l_05
  bone2Name=dyn_hip_tube_ft_l_05
  swing1LimitDegrees=45
  swing1Motion=Free
  swing2LimitDegrees=45
  swing2Motion=Free
  twistLimitDegrees=45
  twistMotion=Free
 }
 {
  bone1Name=dyn_hip_tube_bk_r_04
  bone2Name=dyn_hip_tube_ft_r_04
  swing1LimitDegrees=45
  swing1Motion=Free
  swing2LimitDegrees=45
  swing2Motion=Free
  twistLimitDegrees=45
  twistMotion=Free
 }
 {
  bone1Name=dyn_hip_tube_ft_l_04
  bone2Name=dyn_hip_tube_bk_l_04
  swing1LimitDegrees=45
  swing1Motion=Free
  swing2LimitDegrees=45
  swing2Motion=Free
  twistLimitDegrees=45
  twistMotion=Free
 }
 {
  bone1Name=dyn_hip_tube_bk_r_MSTR
  bone2Name=dyn_waist_tube_r_MSTR
  swing1LimitDegrees=45
  swing1Motion=Free
  swing2LimitDegrees=45
  swing2Motion=Free
  twistLimitDegrees=45
  twistMotion=Free
 }
 {
  bone1Name=dyn_hip_tube_ft_r_03
  bone2Name=dyn_hip_tube_bk_r_04
  swing1LimitDegrees=45
  swing1Motion=Free
  swing2LimitDegrees=45
  swing2Motion=Free
  twistLimitDegrees=45
  twistMotion=Free
 }
]
