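# Secondary-motion AnimBlueprint for the Shock Wave Commando body, animating shoulder pads, tubes, collar and straps over the pose copied from the parent mesh.
# Shoulder pads follow arm rotations or weapon-specific Control Rigs, then constraints position hip tube bones.
# Movement states blend tube Control Rig and RigidBody pairs, feeding collar rig and arm-strap physics.
# Right-arm up/forward values read left-arm bones; leg-strap physics is unconnected.
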AnimBlueprint /BRCosmetics/Characters/Player/Male/Medium/Bodies/M_MED_ShockWave_Commando/Meshes/M_MED_Shock_Wave_Commando_Skeleton_AnimBlueprint
target_skeleton=&/M_MED_Shock_Wave_Commando_Skeleton
graphs [
 AnimGraph {nodes [
   RigidBody() {
    bClampLinearTranslationLimitToRefPose=true
    bOverrideWorldGravity=true
    overridePhysicsAsset=&/M_MED_Shock_Wave_Commando_LegStrapPhys
    simulationSpace=BaseBoneSpace
    alphaBoolBlend {blendOption=Linear}
    baseBoneRef {boneName=pelvis}
    componentAppliedLinearAccClamp=<3000 3000 3000>
    componentLinearAccScale=<0.4 0.4 0.4>
    componentLinearVelScale=<0.15 0.15 0.25>
    overrideWorldGravity {z=-980}
   }
   n1 = Control Rig(Source=Use cached pose 'PostCache'()) {
    alphaBoolBlend {blendOption=Linear}
    controlRigAssetReference {blueprintRigClass=&/M_MED_Shock_Wave_tubes_CtrlRig.M_MED_Shock_Wave_tubes_CtrlRig_C}
   }
   n2 = RigidBody(ComponentPose=Local To Component(LocalPose=@n1)) {
    bClampLinearTranslationLimitToRefPose=true
    bOverrideWorldGravity=true
    overridePhysicsAsset=&/M_MED_Shock_Wave_Commando_Physics
    simulationSpace=BaseBoneSpace
    alphaBoolBlend {blendOption=Linear}
    baseBoneRef {boneName=pelvis}
    componentAppliedLinearAccClamp=<3000 3000 3000>
    componentLinearAccScale=<0.3 0.3 0.3>
    componentLinearVelScale {x=0.3, z=0.3}
    overrideWorldGravity {z=-980}
   }
   n3 = Control Rig(Source=Use cached pose 'PostCache'()) {
    alphaBoolBlend {blendOption=Linear}
    controlRigAssetReference {blueprintRigClass=&/M_MED_Shock_Wave_Tubes_Crouch_CtrlRig.M_MED_Shock_Wave_Tubes_Crouch_CtrlRig_C}
   }
   n4 = RigidBody(ComponentPose=Local To Component(LocalPose=@n3), Alpha=0.6) {
    bClampLinearTranslationLimitToRefPose=true
    bOverrideWorldGravity=true
    overridePhysicsAsset=&M_MED_Shock_Wave_Commando_Physics
    simulationSpace=BaseBoneSpace
    alphaBoolBlend {blendOption=Linear}
    baseBoneRef {boneName=pelvis}
    componentAppliedLinearAccClamp=<3000 3000 3000>
    componentLinearAccScale=<0.45 0.45 0.2>
    componentLinearVelScale=<0.15 0.15 0.1>
    overrideWorldGravity {z=-980}
   }
   n5 = Control Rig(Source=Use cached pose 'PostCache'()) {
    alphaBoolBlend {blendOption=Linear}
    controlRigAssetReference {blueprintRigClass=&/M_MED_Shock_Wave_tubes_run_CtrlRig.M_MED_Shock_Wave_tubes_run_CtrlRig_C}
   }
   n6 = RigidBody(ComponentPose=Local To Component(LocalPose=@n5)) {
    bClampLinearTranslationLimitToRefPose=true
    bOverrideWorldGravity=true
    overridePhysicsAsset=&M_MED_Shock_Wave_Commando_Physics
    simulationSpace=BaseBoneSpace
    alphaBoolBlend {blendOption=Linear}
    baseBoneRef {boneName=pelvis}
    componentAppliedLinearAccClamp=<3000 3000 3000>
    componentLinearAccScale=<0.45 0.45 0.55>
    componentLinearVelScale=<0.45 0.45 0.55>
    overrideWorldGravity {z=-980}
   }
   n7 = Blend Poses by bool(
    BlendPose_0=Component To Local(ComponentPose=@n6)
    BlendPose_1=Blend Poses by bool(
     BlendPose_0=Component To Local(ComponentPose=@n4)
     BlendPose_1=Component To Local(ComponentPose=@n2)
     BlendTime_1=0.9
     bActiveValue=Get bIsCrouching()
    ) {blendType=Linear}
    bActiveValue=Get bIsSprinting()
   ) {blendType=Linear}
   n8 = Control Rig(Source=Use cached pose 'PostCache'()) {
    alphaBoolBlend {blendOption=Linear}
    controlRigAssetReference {blueprintRigClass=&/M_MED_Shock_Wave_tubes_jump_Ctrl.M_MED_Shock_Wave_tubes_jump_Ctrl_C}
   }
   n9 = RigidBody(ComponentPose=Local To Component(LocalPose=@n8)) {
    bClampLinearTranslationLimitToRefPose=true
    bOverrideWorldGravity=true
    overridePhysicsAsset=&M_MED_Shock_Wave_Commando_Physics
    simulationSpace=BaseBoneSpace
    alphaBoolBlend {blendOption=Linear}
    baseBoneRef {boneName=pelvis}
    componentAppliedLinearAccClamp=<3000 3000 3000>
    componentLinearAccScale=<0.45 0.45 0.6>
    componentLinearVelScale=<0.15 0.15 0.25>
    overrideWorldGravity {z=-980}
   }
   n10 = Component To Local(ComponentPose=@n9)
   n11 = OR Boolean(A=Get bIsFalling(), B=Get bIsJumping())
   n12 = Control Rig(Source=Use cached pose 'PostCache'(), Alpha=0.5) {
    alphaBoolBlend {blendOption=Linear}
    controlRigAssetReference {blueprintRigClass=&M_MED_Shock_Wave_tubes_CtrlRig_C}
   }
   n13 = RigidBody(ComponentPose=Local To Component(LocalPose=@n12), Alpha=0.5) {
    bClampLinearTranslationLimitToRefPose=true
    bOverrideWorldGravity=true
    overridePhysicsAsset=&M_MED_Shock_Wave_Commando_Physics
    simulationSpace=BaseBoneSpace
    alphaBoolBlend {blendOption=Linear}
    baseBoneRef {boneName=pelvis}
    componentAppliedLinearAccClamp=<3000 3000 3000>
    componentLinearAccScale=<0.3 0.3 0.3>
    componentLinearVelScale {x=0.3, z=0.3}
    overrideWorldGravity {z=-980}
   }
   n14 = Control Rig(Source=Blend Poses by bool(
     BlendPose_0=Component To Local(ComponentPose=@n13)
     BlendPose_1=Blend Poses by bool(BlendPose_0=@n10, BlendPose_1=@n7, bActiveValue=@n11) {blendType=Linear}
     bActiveValue=Get bIsSkyDiving()
    ) {blendType=Linear}) {
    alphaBoolBlend {blendOption=Linear}
    controlRigAssetReference {blueprintRigClass=&/M_MED_Shock_Wave_Collar_CtrlRig.M_MED_Shock_Wave_Collar_CtrlRig_C}
   }
   n15 = RigidBody(ComponentPose=Local To Component(LocalPose=@n14)) {
    bClampLinearTranslationLimitToRefPose=true
    bOverrideWorldGravity=true
    overridePhysicsAsset=&/M_MED_Shock_Wave_Commando_ArmStrapPhys
    simulationSpace=BaseBoneSpace
    alphaBoolBlend {blendOption=Linear}
    baseBoneRef {boneName=spine_04}
    componentAppliedLinearAccClamp=<3000 3000 3000>
    componentLinearAccScale=<0.3 0.3 0.3>
    componentLinearVelScale=<0.15 0.15 0.25>
    overrideWorldGravity {z=-980}
   }
   Output Pose(Result=Component To Local(ComponentPose=@n15)) {layerGroup=None}
   Save cached pose 'AnimCache'(Pose=Copy Pose From Mesh() {bUseAttachedParent=true}) {cacheName=AnimCache}
   n16 = "Transform (Modify) Bone - Bone: dyn_shldr_pad_pivot_base_l"(
    ComponentPose=Local To Component(LocalPose=Use cached pose 'AnimCache'())
    Alpha=Get armUp_L() -> (armUp_L:double)
    Translation="9,-1,1.5"
    Rotation="0,0,0"
    Scale="1,1,1"
   ) {
    rotationMode=BMM_Additive
    rotationSpace=BCS_BoneSpace
    translationMode=BMM_Additive
    translationSpace=BCS_BoneSpace
    alphaBoolBlend {blendOption=Linear}
    alphaScaleBiasClamp {bMapRange=true, inRange {max=70}}
    boneToModify {boneName=dyn_shldr_pad_pivot_base_l}
    translation=<9 -1 1.5>
   }
   n17 = "Transform (Modify) Bone - Bone: dyn_shldr_pad_pivot_base_l"(ComponentPose=@n16, Alpha=Get armFwd_L() -> (armFwd_L:double), Translation="0,0,1", Rotation="0,0,-15", Scale="1,1,1") {
    rotationMode=BMM_Additive
    rotationSpace=BCS_BoneSpace
    translationMode=BMM_Additive
    translationSpace=BCS_BoneSpace
    alphaBoolBlend {blendOption=Linear}
    alphaScaleBiasClamp {bMapRange=true, inRange {max=70}}
    boneToModify {boneName=dyn_shldr_pad_pivot_base_l}
    rotation {roll=-15}
    translation {z=1}
   }
   n18 = "Transform (Modify) Bone - Bone: dyn_shldr_pad_pivot_base_l"!(
    ComponentPose=@n17
    bAlphaBoolEnabled:bool=Get isMeleeWeapon() -> (isMeleeWeapon:bool)
    Translation="0,0,0"
    Rotation="0,0,-20"
    Scale="1,1,1"
   ) -> (Pose) {
    alphaInputType=Bool
    rotationMode=BMM_Additive
    rotationSpace=BCS_BoneSpace
    translationMode=BMM_Additive
    translationSpace=BCS_BoneSpace
    alphaBoolBlend {blendOption=Linear}
    alphaScaleBiasClamp {inRange {max=70}}
    boneToModify {boneName=dyn_shldr_pad_pivot_base_l}
    rotation {roll=-20}
   }
   n19 = "Transform (Modify) Bone - Bone: dyn_shldr_pad_pivot_base_r"(ComponentPose=@n18, Alpha=Get armUp_R() -> (armUp_R:double), Translation="0,0,0", Rotation="0,0,0", Scale="1,1,1") {
    rotationMode=BMM_Additive
    rotationSpace=BCS_BoneSpace
    translationMode=BMM_Additive
    translationSpace=BCS_BoneSpace
    alphaBoolBlend {blendOption=Linear}
    alphaScaleBiasClamp {bMapRange=true, interpSpeedDecreasing=1, interpSpeedIncreasing=1, inRange {max=70}}
    boneToModify {boneName=dyn_shldr_pad_pivot_base_r}
   }
   n20 = "Transform (Modify) Bone - Bone: dyn_shldr_pad_pivot_base_r"(ComponentPose=@n19, Alpha=Get armFwd_R() -> (armFwd_R:double), Translation="0,0,0", Rotation="0,0,50", Scale="1,1,1") {
    rotationMode=BMM_Additive
    rotationSpace=BCS_BoneSpace
    translationMode=BMM_Additive
    translationSpace=BCS_BoneSpace
    alphaBoolBlend {blendOption=Linear}
    alphaScaleBiasClamp {bMapRange=true, interpSpeedDecreasing=1, interpSpeedIncreasing=1, inRange {max=70}}
    boneToModify {boneName=dyn_shldr_pad_pivot_base_r}
    rotation {roll=50}
   }
   n21 = Control Rig(Source=Use cached pose 'AnimCache'()) {
    alphaBoolBlend {blendOption=Linear}
    controlRigAssetReference {blueprintRigClass=&/M_MED_Shock_Wave_Rifle_CtrlRig.M_MED_Shock_Wave_Rifle_CtrlRig_C}
   }
   n22 = Control Rig(Source=Use cached pose 'AnimCache'()) {
    alphaBoolBlend {blendOption=Linear}
    controlRigAssetReference {blueprintRigClass=&/M_MED_Shock_Wave_Pistol_CtrlRig.M_MED_Shock_Wave_Pistol_CtrlRig_C}
   }
   n23 = Blend Poses by bool(
    BlendPose_0=@n22
    BlendPose_1=Blend Poses by bool(BlendPose_0=@n21, BlendPose_1=Component To Local(ComponentPose=@n20), bActiveValue=Get isAimingRifle() -> (isAimingRifle:bool)) {blendType=Linear}
    bActiveValue=Get isPistolADS() -> (isPistolADS:bool)
   ) {blendType=Linear}
   n24 = "Transform (Modify) Bone - Bone: dyn_shldr_pad_pivot_base_l"!(
    ComponentPose=Local To Component(LocalPose=Use cached pose 'AnimCache'())
    Alpha:float=Get armSky_l() -> (armSky_l:double)
    Translation="7,0,0"
    Rotation="0,0,0"
    Scale="1,1,1"
   ) -> (Pose) {
    rotationMode=BMM_Additive
    rotationSpace=BCS_BoneSpace
    translationMode=BMM_Additive
    translationSpace=BCS_BoneSpace
    alphaBoolBlend {blendOption=Linear}
    alphaScaleBiasClamp {bMapRange=true, inRange {max=70}}
    boneToModify {boneName=dyn_shldr_pad_pivot_base_l}
    translation {x=7}
   }
   n25 = "Transform (Modify) Bone - Bone: dyn_shldr_pad_pivot_base_r"(ComponentPose=@n24, Alpha=Get armSky_r() -> (armSky_r:double), Translation="-7,0,0", Rotation="0,0,0", Scale="1,1,1") {
    rotationMode=BMM_Additive
    rotationSpace=BCS_BoneSpace
    translationMode=BMM_Additive
    translationSpace=BCS_BoneSpace
    alphaBoolBlend {blendOption=Linear}
    alphaScaleBiasClamp {bMapRange=true, interpSpeedDecreasing=1, interpSpeedIncreasing=1, inRange {max=70}}
    boneToModify {boneName=dyn_shldr_pad_pivot_base_r}
    translation {x=-7}
   }
   n26 = Component To Local(ComponentPose=@n25)
   n27 = Get bIsSkyDiving()
   n28 = "Constraint - dyn_hip_tube_bk_r_MSTR (PPP)"!(
    ComponentPose=Local To Component(LocalPose=Blend Poses by bool(BlendPose_0=@n26, BlendPose_1=@n23, bActiveValue=@n27) {blendType=Linear})
    Alpha=1
    ConstraintWeights_0=0.8069
    ConstraintWeights_1:float=0.1862
    ConstraintWeights_2:float=0.0069
   ) -> (Pose) {
    alphaBoolBlend {blendOption=Linear}
    boneToModify {boneName=dyn_hip_tube_bk_r_MSTR}
    constraintSetup [{targetBone {boneName=pelvis}}, {targetBone {boneName=thigh_twist_01_r}}, {targetBone {boneName=spine_01}}]
    constraintWeights [1,1,1]
   }
   n29 = "Constraint - dyn_hip_tube_bk_l_MSTR (PPP)"!(ComponentPose=@n28, Alpha=1, ConstraintWeights_0=0.787, ConstraintWeights_1:float=0.1629, ConstraintWeights_2:float=0.0492) -> (Pose) {
    alphaBoolBlend {blendOption=Linear}
    boneToModify {boneName=dyn_hip_tube_bk_l_MSTR}
    constraintSetup [{targetBone {boneName=pelvis}}, {targetBone {boneName=thigh_twist_01_l}}, {targetBone {boneName=spine_01}}]
    constraintWeights [1,1,1]
   }
   n30 = "Constraint - dyn_tube_front_r_MSTR (PPP)"!(ComponentPose=@n29, Alpha=1, ConstraintWeights_0=0.6892, ConstraintWeights_1:float=0.1888, ConstraintWeights_2:float=0.122) -> (Pose) {
    alphaBoolBlend {blendOption=Linear}
    boneToModify {boneName=dyn_tube_front_r_MSTR}
    constraintSetup [{targetBone {boneName=pelvis}}, {targetBone {boneName=thigh_twist_01_r}}, {targetBone {boneName=spine_01}}]
    constraintWeights [1,1,1]
   }
   n31 = "Constraint - dyn_tube_front_l_MSTR (PPP)"!(ComponentPose=@n30, Alpha=1, ConstraintWeights_0=0.6819, ConstraintWeights_1:float=0.2191, ConstraintWeights_2:float=0.0988) -> (Pose) {
    alphaBoolBlend {blendOption=Linear}
    boneToModify {boneName=dyn_tube_front_l_MSTR}
    constraintSetup [{targetBone {boneName=pelvis}}, {targetBone {boneName=thigh_twist_01_l}}, {targetBone {boneName=spine_01}}]
    constraintWeights [1,1,1]
   }
   Save cached pose 'PostCache'(Pose=Component To Local(ComponentPose=@n31)) {cacheName=PostCache}
   Get armFwd_L()
   Get armUp_R()
  ]}
 EventGraph {nodes [
   Event BlueprintUpdateAnimation()
   > n1 = Sequence!(execute) -> (then_0, then_1, then_2:Exec, then_3:Exec)
   > n2 = Sequence!(execute) -> (then_0, then_1, then_2, then_3, then_4:Exec, then_5:Exec, then_6:Exec)
   n3 = GetDeltaTransformFromRefPose(self=GetOwnerSkeletalMeshComponent(), BoneName=upperarm_l, BaseName=clavicle_l)
   Set armSky_l(execute=@n2.then_4, armSky_l:double=BreakRotator(InRot=BreakTransform(InTransform=@n3).Rotation).Pitch) -> (Output_Get:double)
   n4 = GetDeltaTransformFromRefPose(self=GetOwnerSkeletalMeshComponent(), BoneName=upperarm_r, BaseName=clavicle_r)
   Set armSky_r(execute=@n2.then_5, armSky_r:double=BreakRotator(InRot=BreakTransform(InTransform=@n4).Rotation).Pitch) -> (Output_Get:double)
   n5 = GetDeltaTransformFromRefPose(self=GetOwnerSkeletalMeshComponent(), BoneName=thigh_l, BaseName=pelvis)
   Set legUp_L(execute=@n2.then_1, legUp_L:double=BreakRotator(InRot=BreakTransform(InTransform=@n5).Rotation).Yaw) -> (Output_Get:double)
   n6 = GetOwnerSkeletalMeshComponent()
   n7 = BreakRotator(InRot=BreakTransform(InTransform=GetDeltaTransformFromRefPose(self=@n6, BoneName=upperarm_l, BaseName=clavicle_l)).Rotation)
   Set armUp_L(execute=@n2.then_2, armUp_L:double=@n7.Pitch) -> (Output_Get:double)
   > Set armFwd_L(armFwd_L:double=@n7.Yaw) -> (Output_Get:double)
   n8 = GetOwnerSkeletalMeshComponent()
   n9 = BreakRotator(InRot=BreakTransform(InTransform=GetDeltaTransformFromRefPose(self=@n8, BoneName=upperarm_l, BaseName=clavicle_l)).Rotation)
   Set armUp_R(execute=@n2.then_3, armUp_R:double=@n9.Pitch) -> (Output_Get:double)
   > Set armFwd_R(armFwd_R:double=@n9.Yaw) -> (Output_Get:double)
   n10 = GetDeltaTransformFromRefPose(self=GetOwnerSkeletalMeshComponent(), BoneName=thigh_r, BaseName=pelvis)
   Set legUp_R(execute=@n2.then_6, legUp_R:double=BreakRotator(InRot=BreakTransform(InTransform=@n10).Rotation).Pitch) -> (Output_Get:double)
   n11 = Cast To FortPlayerAnimInstance(execute=@n1.then_2, Object=GetAnimInstance(self=GetOwnerSkeletalMeshComponent()))
   > Set isPistolADS(isPistolADS:bool=false) -> (Output_Get:bool)
   > Set isAimingRifle(isAimingRifle:bool=false) -> (Output_Get:bool)
   > n12 = Switch on EFortWeaponCoreAnimation!(execute, Selection:EFortWeaponCoreAnimation Enum=Get WeaponCoreAnim(self=@n11)) -> (
    Melee:Exec
    Pistol:Exec
    Shotgun:Exec
    PaperBlueprint:Exec
    Rifle:Exec
    MeleeOneHand:Exec
    MachinePistol:Exec
    RocketLauncher:Exec
    GrenadeLauncher:Exec
    GoingCommando:Exec
    AssaultRifle:Exec
    TacticalShotgun:Exec
    SniperRifle:Exec
    TrapPlacement:Exec
    ShoulderLauncher:Exec
    AbilityDecoTool:Exec
    Crossbow:Exec
    C4:Exec
    RemoteControl:Exec
    DualWield:Exec
    AR_BullPup:Exec
    AR_ForwardGrip:Exec
    MedPackPaddles:Exec
    SMG_P90:Exec
    AR_DrumGun:Exec
    Consumable_Small:Exec
    Consumable_Large:Exec
    Balloon:Exec
    MountedTurret:Exec
    CreativeTool:Exec
    ExplosiveBow:Exec
    AshtonIndigo:Exec
    AshtonChicago:Exec
    MeleeDualWield:Exec
    Unarmed:Exec
   )
   Branch(execute=[
     @n12.AssaultRifle
     @n12.Crossbow
     @n12.GrenadeLauncher
     @n12.Rifle
     @n12.RocketLauncher
     @n12.Shotgun
     @n12.SniperRifle
     @n12.TacticalShotgun
    ])
   > Set isAimingRifle(isAimingRifle=true)
   Branch(execute=[@n12.MachinePistol, @n12.Pistol])
   > Set isPistolADS(isPistolADS=true)
   n13 = Cast To FortPlayerAnimInstance(execute=@n1.then_1, Object=GetAnimInstance(self=GetOwnerSkeletalMeshComponent()))
   n14 = Get WeaponCoreAnim(self=@n13)
   > Set isMeleeWeapon(isMeleeWeapon:bool="Equal Exactly (String)"(A=Melee, B=Enum to String!(Enumerator:EFortWeaponCoreAnimation Enum=@n14) -> (ReturnValue:String))) -> (Output_Get:bool)
   TryGetPawnOwner()
   Set isAimingPistol(isAimingPistol:double=1) -> (Output_Get:double)
   Set isAimingPistol(isAimingPistol=0)
  ]}
]
variables [
 armFwd_L {type=number}
 armFwd_R {type=number}
 armSky_l {type=number}
 armSky_r {type=number}
 armUp_L {type=number}
 armUp_R {type=number}
 isAimingPistol {type=number}
 isAimingRifle {type=boolean}
 isMeleeWeapon {type=boolean}
 isPistolADS {type=boolean}
 legUp_L {type=number}
 legUp_R {type=number}
]
